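# Animation Blueprint for the face accessory of the StaminaVigor Super male player body, driving hair bones that follow the parent mesh.
# It copies the attached parent's pose, then additively adjusts HairLeft4 by an ADS UP alpha computed from head movement relative to neck_02.
# A head-based RigidBody node then simulates secondary motion using the base StaminaVigor physics asset.
# A sine-based Ambient Animation hair sway exists, but its update calls and bone nodes are unconnected.
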AnimBlueprint /BRCosmetics/Characters/Player/Male/Medium/Bodies/M_MED_StaminaVigor_Super/Meshes/Parts/M_MED_StaminaVigor_Super_FaceAcc_AnimBP
target_skeleton=&/M_MED_StaminaVigor_Super_FaceAcc_Skeleton
graphs [
 Ambient Animation {nodes [
   n1 = Ambient Animation!() -> (
    then
    Time Offset 1:double=~
    Magnitude Multiplier:double=~
    Period:Vector=~
    Magnitude:Vector="10,0,0"
    Period Multiplier:double=1
    Rotate Multiplier:double=~
   ) {is=title K2Node_FunctionEntry}
   n2 = BreakVector(InVec=@n1.Magnitude)
   n3 = BreakVector(InVec=@n1.Period)
   n4 = GetGameTimeInSeconds()
   n5 = float * float(A=float * float(A=@n1.Period Multiplier, B=@n3.Y), B=@n4)
   n6 = float * float(A="Sin (Radians)"(A=float + float(A=@n5, B=@n1) {is=title K2Node_CommutativeAssociativeBinaryOperator}), B=@n2.Y)
   n7 = float * float(A=float * float(A=@n1.Period Multiplier, B=@n3.Z), B=@n4)
   n8 = float * float(A="Sin (Radians)"(A=float + float(A=@n7, B=@n1)), B=@n2.Z)
   n9 = float * float(A=float * float(A=@n1.Period Multiplier, B=@n3), B=@n4)
   n10 = float * float(A="Sin (Radians)"(A=float + float(A=@n9, B=@n1)), B=@n2)
   n11 = MakeRotator(Roll=@n10, Pitch=@n6, Yaw=@n8)
   > Return Node!(
    execute
    VectorOut:Vector=vector * float(A=MakeVector(X=@n10, Y=@n6, Z=@n8), B=@n1.Magnitude Multiplier)
    RotateOut:Rotator="Scale Rotator (Float)"!(A:Rotator="Scale Rotator (Float)"(A=@n11, B=@n1.Magnitude Multiplier), B:float=@n1.Rotate Multiplier) -> (ReturnValue:Rotator)
   )
  ]}
 AnimGraph {nodes [
   n1 = "Transform (Modify) Bone - Bone: Dyn_Master_HairMain"(Alpha=1, Translation="0,0,0", Rotation=Get Ambient Rot L1() -> (Ambient Rot L1:Rotator), Scale="1,1,1") {
    rotationMode=BMM_Replace
    rotationSpace=BCS_BoneSpace
    translationSpace=BCS_BoneSpace
    alphaBoolBlend {blendOption=Linear}
    boneToModify {boneName=Dyn_Master_HairMain}
   }
   n2 = "Transform (Modify) Bone - Bone: Dyn_Master_HairRight1"(ComponentPose=@n1, Alpha=1, Translation="0,0,0", Rotation=Get Ambient Rot R1() -> (Ambient Rot R1:Rotator), Scale="1,1,1") {
    rotationMode=BMM_Replace
    rotationSpace=BCS_BoneSpace
    translationSpace=BCS_BoneSpace
    alphaBoolBlend {blendOption=Linear}
    boneToModify {boneName=Dyn_Master_HairRight1}
   }
   "Transform (Modify) Bone - Bone: Dyn_Master_HairRight2"(ComponentPose=@n2, Alpha=1, Translation="0,0,0", Rotation=Get Ambient Rot R2() -> (Ambient Rot R2:Rotator), Scale="1,1,1") {
    rotationMode=BMM_Additive
    rotationSpace=BCS_BoneSpace
    translationSpace=BCS_BoneSpace
    alphaBoolBlend {blendOption=Linear}
    boneToModify {boneName=Dyn_Master_HairRight2}
   }
   n3 = "Transform (Modify) Bone - Bone: Dyn_Master_HairLeft4"(
    ComponentPose=Local To Component(LocalPose=Copy Pose From Mesh() {bUseAttachedParent=true})
    Alpha=Get ADS UP() -> (ADS UP:double)
    Translation="0,0,0"
    Rotation="0,-18,0"
    Scale="0.8,1,1"
   ) {
    rotationMode=BMM_Additive
    rotationSpace=BCS_BoneSpace
    scaleMode=BMM_Additive
    scaleSpace=BCS_BoneSpace
    translationMode=BMM_Additive
    translationSpace=BCS_BoneSpace
    alphaBoolBlend {blendOption=Linear}
    boneToModify {boneName=Dyn_Master_HairLeft4}
    rotation {yaw=-18}
    scale {x=0.8}
   }
   n4 = RigidBody(ComponentPose=@n3) {
    bClampLinearTranslationLimitToRefPose=true
    bFreezeIncomingPoseOnStart=true
    bOverrideWorldGravity=true
    overridePhysicsAsset=/BRCosmetics/Characters/Player/Male/Medium/Bodies/M_MED_StaminaVigor/Meshes/Parts/M_MED_StaminaVigor_FaceAcc_Physics
    simulationSpace=BaseBoneSpace
    alphaBoolBlend {blendOption=Linear}
    baseBoneRef {boneName=head}
    componentAppliedLinearAccClamp=<3000 3000 3000>
    componentLinearAccScale=<0.2 0.2 0.2>
    componentLinearVelScale=<0.2 0.2 0.2>
    overrideWorldGravity {z=-980}
   }
   Output Pose(Result=Component To Local(ComponentPose=@n4)) {layerGroup=None}
  ]}
 EventGraph {nodes [
   n1 = Ambient Animation!(
    execute
    self
    Time Offset 1:double=0
    Magnitude Multiplier:double=1
    Period:Vector="7,0,0"
    Magnitude:Vector="2,0,0"
    Period Multiplier:double=1
    Rotate Multiplier:double=1
   ) -> (then, VectorOut:Vector, RotateOut:Rotator) {is=title K2Node_CallFunction}
   > Set Ambient Rot R1(Ambient Rot R1:Rotator=MakeRotator(Yaw=BreakRotator(InRot=@n1.RotateOut))) -> (Output_Get:Rotator)
   n2 = Ambient Animation(Time Offset 1=0, Magnitude Multiplier=1, Period="10,0,5", Magnitude="1,0,2", Period Multiplier=1, Rotate Multiplier=1)
   > Set Ambient Rot R2(Ambient Rot R2:Rotator=MakeRotator(Yaw=BreakRotator(InRot=@n2.RotateOut))) -> (Output_Get:Rotator)
   n3 = Ambient Animation(Time Offset 1=0, Magnitude Multiplier=1, Period="10,0,0", Magnitude="1,0,0", Period Multiplier=1, Rotate Multiplier=1)
   > Set Ambient Rot L1(Ambient Rot L1:Rotator=MakeRotator(Yaw=BreakRotator(InRot=@n3.RotateOut))) -> (Output_Get:Rotator)
   Event BlueprintUpdateAnimation()
   > Sequence!(execute) -> (then_0, then_1, then_2:Exec, then_3:Exec)
   > Set ADS UP(ADS UP:double=GetAlphaFromReferenceTransformDelta(
     SkeletalMeshComponent=Get Skeletal Mesh() -> (Skeletal Mesh:Skeletal Mesh Component Object Reference)
     BoneName=head
     BaseName=neck_02
     InputRange=(X=-8,Y=-20)
     TwistAxis="0,0,0"
    )) -> (Output_Get:double)
   Event BlueprintInitializeAnimation()
   > Set Skeletal Mesh(Skeletal Mesh:Skeletal Mesh Component Object Reference=GetOwnerSkeletalMeshComponent()) -> (Output_Get:Skeletal Mesh Component Object Reference)
  ]}
]
variables [
 ADS UP {type=number}
 Ambient Rot L1 {type=Rotator}
 Ambient Rot R1 {type=Rotator}
 Ambient Rot R2 {type=Rotator}
 Skeletal Mesh {type=SkeletalMeshComponent}
 Time Offset 1 {type=number}
]
